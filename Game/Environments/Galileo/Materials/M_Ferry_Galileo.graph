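# Nanite-enabled lit material for the Galileo ferry environment prop, reusing the ferry glider's textures with optional damage, swamp and snow looks.
# By default, noise-broken height lerps paint burnt areas from vertex red and fuel leaks with Fresnel glow from vertex green.
# The "Is in Swamp" static switch replaces both with tinted, glossy wetness below a world height.
# Seasonal snow then blends by upward normals, minimap mask and vertex alpha before the game-wide shared function.
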Material /Game/Environments/Galileo/Materials/M_Ferry_Galileo
graph {
 properties {bUseMaterialAttributes=true, bUsedWithNanite=true, pixelDepthOffsetMode=PDOM_Legacy, usedShadingModels=Default Lit}
 nodes [
  n1 = TextureSampleParameter2D() {
   parameterName=Diffuse
   texture=/Game/Gadgets/Assets/VinderTech_GliderChute/Glider_Galileo_Ferry/Textures/T_Galileo_Ferry_Galileo_Ferry_Body_D
  }
  n2 = TextureSampleParameter2D() {
   parameterName=SpecularMasks
   samplerType=SAMPLERTYPE_Masks
   texture=/Game/Gadgets/Assets/VinderTech_GliderChute/Glider_Galileo_Ferry/Textures/T_Galileo_Ferry_Galileo_Ferry_Body_S
  }
  n3 = TextureSampleParameter2D() {
   parameterName=Normals
   samplerType=SAMPLERTYPE_Normal
   texture=/Game/Gadgets/Assets/VinderTech_GliderChute/Glider_Galileo_Ferry/Textures/T_Galileo_Ferry_Galileo_Ferry_Body_N
  }
  n4 = TextureSample(UVs=Multiply(A=TextureCoordinate(), B=ScalarParameter() {defaultValue=5, parameterName=DamageNoise})) {samplerType=SAMPLERTYPE_Normal, texture=/Game/Effects/Fort_Effects/Textures/Tile/noise_flamy_Tile_NRM}
  n5 = Constant3Vector() {constant=<rgba 0.5 0.5 0 0>}
  n6 = Normalize(VectorInput=Add(A=@n3.RGB, B=Multiply(A=@n4.RGB, B=@n5)))
  n7 = SetMaterialAttributes!(MaterialAttributes, Base Color=@n1.RGB, Metallic=@n2.G, Specular=@n2.G, Roughness=@n2.G, Normal=@n6) {attributeSetTypes [
    #1
    #2
    #3
    #4
    #5
   ]}
  n8 = GetMaterialAttributes(None=@n7) {attributeGetTypes [#1]}
  n9 = Multiply(A=@n8.Base Color, B=VectorParameter() {group=Swamp, parameterName=WaterTint, defaultValue=<rgba 0.2448 0.2131 0.0492 0>}.RGB)
  n10 = ScalarParameter() {defaultValue=0.12, group=Swamp, parameterName=Swamp R}
  n11 = WorldPosition()
  n12 = ScalarParameter() {defaultValue=-3903, group=Swamp, parameterName=SwampHeight}
  n13 = Divide(
   A=ComponentMask(None=Subtract(A=@n11.XYZ, B=@n12)) {b=true}
   B=ScalarParameter() {defaultValue=155.6466, group=Swamp, parameterName=WetnessFade}
  )
  n14 = Multiply(A=TextureCoordinate(), B=ScalarParameter() {defaultValue=5, parameterName=DamageNoise})
  n15 = TextureSample(UVs=@n14) {samplerType=SAMPLERTYPE_Normal, texture=&noise_flamy_Tile_NRM}
  n16 = Constant3Vector() {constant=<rgba 1 1 0 0>}
  n17 = SetMaterialAttributes!(
   MaterialAttributes
   Base Color=VectorParameter() {group=Burn, parameterName=GroundTone, defaultValue=<rgba 0.0156 0.0156 0.0156 0>}.RGB
   Specular=Constant() {r=0.1}
   Roughness=ScalarParameter() {defaultValue=1.5, parameterName=Param_4}
   Normal=Add(A=Multiply(A=@n15.RGB, B=@n16), B=@n3.RGB)
   Metallic=Constant()
  ) {attributeSetTypes [
    #1
    #3
    #4
    #5
    #2
   ]}
  n18 = OneMinus(None=VertexColor().R)
  n19 = TextureSample(UVs=@n14) {texture=/Game/Global/Textures/Default/Noise/T_Noise_Cloud}
  n20 = Multiply(A=@n19.R, B=ScalarParameter() {defaultValue=1, parameterName=Param_2})
  n21 = ScalarParameter() {defaultValue=0.3565}
  n22 = GetMaterialAttributes(None=BlendMaterialAttributes(
    A=@n7
    B=@n17
    Alpha=MaterialFunctionCall(Transition Phase=@n18, Height Texture=@n20, Contrast=@n21) {materialFunction=&HeightLerp}.Alpha
   )) {attributeGetTypes [#6, #1, #4]}
  n23 = VectorParameter() {group=Leak, parameterName=Fuel Color, defaultValue=<rgba 0.0045 0.0781 0.0711 0>}
  n24 = MaterialFunctionCall(
   A=@n22.Base Color
   B=MaterialFunctionCall(Base=@n22.Base Color, Blend=@n23.RGB) {materialFunction=&Blend_Overlay}.Result
   Transition Phase=OneMinus(None=VertexColor().G)
   Height Texture=Multiply(A=@n22.Opacity, B=OneMinus(None=@n19.R))
  ) {materialFunction=&HeightLerp}
  n25 = ScalarParameter() {defaultValue=0.25, group=Leak, parameterName=Fuel R}
  n26 = LinearInterpolate(B=@n23.RGB, Alpha=Multiply(A=@n24.Lerp Alpha No Contrast, B=Fresnel(Normal=@n6) {baseReflectFraction=0.5, exponent=3}))
  n27 = ScalarParameter() {defaultValue=0.2, group=Leak, parameterName=Fuel Emissive}
  n28 = StaticSwitchParameter(
   True=BlendMaterialAttributes(
    A=SetMaterialAttributes!(MaterialAttributes=@n8.MaterialAttributes, Base Color=@n9, Roughness=@n10) {attributeSetTypes [#1, #4]}
    B=@n7
    Alpha=Saturate(None=@n13)
   )
   False=SetMaterialAttributes!(
    MaterialAttributes=@n22.MaterialAttributes
    Base Color=@n24.Results
    Roughness=LinearInterpolate(A=@n22.Roughness, B=@n25, Alpha=@n24.Alpha)
    Emissive Color=Multiply(A=@n26, B=@n27)
   ) {attributeSetTypes [#1, #4, #7]}
  ) {group=Swamp, parameterName=Is in Swamp}
  n29 = MaterialFunctionCall!(
   IsStaticSnow
   UseAlternateStaticSnowMask
   AlternateStaticSnowMask
   "UseRoad&&OcclusionMask"
   IsForGrassMeshToPushDown
   AdditionalMaskToRemoveSnow
   IsFallingSnowFX
   UseMinimapCoords
   IncreaseSnowUsingSlope
   UseSmoothBorder
   LerpHighM
   TemporalHeightLerpContrast
   IsLobbyIsland
   UseBR-IslandMask
   IsTree
   IsForRenderTarget
   UseSparkle
   UseSnowNormals
   UseAdditionalMaskToRemoveSnow
   IsTangentSpaceInput
   UseVertexNormalsForBasicSnow
   UseTangentSnowNormals
   Material=@n28
  ) {materialFunction=/Game/Athena/Apollo/Environments/Landscape/Materials/Biome/UniqueBiomeMaterialFunctions/MF_Apollo_Snow_Progression}
  n30 = Transform(None=GetMaterialAttributes(None=@n28) {attributeGetTypes [#5]}.Normal)
  n31 = Multiply(A=Saturate(None=ComponentMask(None=@n30) {b=true}), B=ScalarParameter() {defaultValue=0.22, parameterName=Snow Amount})
  n32 = Multiply(A=@n29.MinimapMask, B=VertexColor().A)
  n33 = Constant() {r=1}
  MaterialFunctionCall!(In=BlendMaterialAttributes(
    A=@n28
    B=@n29.Material
    Alpha=MaterialFunctionCall(Transition Phase=@n31, Height Texture=@n32, Contrast=@n33) {materialFunction=&HeightLerp}.Alpha
   )) {materialFunction=/Game/Athena/WinterArt/Materials/MF_GameWideShared}
 ]
}
