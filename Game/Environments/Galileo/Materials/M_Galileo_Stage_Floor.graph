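# Default Lit material for the Galileo stage floor, built as a brushed, highly metallic surface with broken-up tone and roughness.
# A brushed anisotropy mask, with tiling scaled by object radius, combines with hull breakup texture sampled at two scales.
# Vertex color red weights the finer breakup.
# These masks blend grayscale diffuse and roughness parameters, and tiling noise blends in a third roughness value.
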Material /Game/Environments/Galileo/Materials/M_Galileo_Stage_Floor
graph {
 properties {pixelDepthOffsetMode=PDOM_Legacy, usedShadingModels=Default Lit}
 nodes [
  n1 = TextureCoordinate() {uTiling=0.5, vTiling=0.5}
  n2 = ObjectRadius()
  n3 = ScalarParameter() {defaultValue=0.01, parameterName=BrushedPatternTiling}
  n4 = TextureSample(UVs=Multiply(A=@n1, B=Multiply(A=@n2, B=@n3))) {samplerType=SAMPLERTYPE_LinearGrayscale, texture=/Game/Environments/Sets/MilitaryBase/Textures/T_SimpleAnsiotrophy_M}
  Add(A=@n4.R) {constB=0.5}
  Add()
  Constant() {r=0.5}
  n5 = TextureCoordinate() {uTiling=2, vTiling=2}
  n6 = LinearInterpolate(
   A=TextureSample(UVs=@n1) {texture=/Game/Environments/Sets/Pirate_Cove/Pirate_Ship/Textures/T_Hull_Breakup}.R
   B=LinearInterpolate(B=TextureSample(UVs=@n5) {texture=&T_Hull_Breakup}.R, Alpha=VertexColor().R)
  )
  LinearInterpolate(
   A=ScalarParameter() {defaultValue=0.1, parameterName=Diffuse_A}
   B=ScalarParameter() {defaultValue=0.4, parameterName=Diffuse_B}
   Alpha=LinearInterpolate(A=LinearInterpolate(Alpha=@n4.R) {constA=0.85, constB=1.15}, B=@n6)
  )
  n7 = LinearInterpolate(A=@n4.R, B=@n6) {constAlpha=0.75}
  n8 = ScalarParameter() {defaultValue=1.1, parameterName=RoughnessBreakup}
  LinearInterpolate(
   A=LinearInterpolate(
    A=ScalarParameter() {defaultValue=0.45, parameterName=Roughness_A}
    B=ScalarParameter() {parameterName=Roughness_B}
    Alpha=Saturate(None=@n7)
   )
   B=ScalarParameter() {defaultValue=0.28, parameterName=Roughness_C}
   Alpha=TextureSample(UVs=Multiply(A=@n5, B=@n8)) {samplerType=SAMPLERTYPE_Grayscale, texture=/Game/Global/Textures/Default/Noise/TilingNoise06}.R
  )
  ScalarParameter() {defaultValue=0.95, parameterName=Metallic}
 ]
}
