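# Master Default Lit material for the Galileo stage environment, with parameterized diffuse, normal, spec/metal/rough and emissive textures, also usable on Niagara mesh particles.
# Glow multiplies the emissive texture's blue channel by a tint color and a strength that can add StageEmissive from the MPC_Galileo parameter collection.
# A timestamp-driven flash scales that glow, easing time since FlashStartTime over FlashDuration on a clock with a five-second period.
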Material /Game/Environments/Galileo/Materials/M_Galileo_Stage_Master
graph {
 properties {bUsedWithNiagaraMeshParticles=true, pixelDepthOffsetMode=PDOM_Legacy, usedShadingModels=Default Lit}
 nodes [
  n1 = Subtract(A=Time() {period=5}, B=ScalarParameter() {parameterName=FlashStartTime})
  n2 = ScalarParameter() {defaultValue=0.75, group=Timestamp, parameterName=FlashDuration}
  n3 = MaterialFunctionCall!(In=Clamp(None=Divide(A=@n1, B=@n2))) {materialFunction=/Game/Effects/Fort_Effects/Materials/_Functions/MF_Smoothstep_EaseOut}
  n4 = ScalarParameter() {defaultValue=4, parameterName=FlashBrightness}
  n5 = Add(
   A=LinearInterpolate(
    B=CollectionParameter() {collection=/Game/Athena/Prototype/Blueprints/Galileo/Blueprints/MPC_Galileo, parameterName=StageEmissive}
    Alpha=ScalarParameter() {parameterName=UseMPCEmissive}
   )
   B=ScalarParameter() {defaultValue=2, parameterName=Emissive_Strength}
  )
  n6 = TextureSampleParameter2D() {parameterName=Emissive, texture=/Game/Environments/Galileo/Textures/T_Galileo_Stage_Cylinder_E}
  Multiply(
   A=Add(A=Multiply(A=@n3.Result, B=@n4))
   B=Multiply(A=VectorParameter() {parameterName=EmissiveColor, defaultValue=<rgba 0.0246 0.2427 0.5 1>}.RGB, B=Multiply(A=@n5, B=@n6.B))
  )
  Saturate(None=Multiply(B=TextureSampleParameter2D() {parameterName=Diffuse, texture=/Game/Environments/Galileo/Textures/T_Galileo_Stage_Cylinder_D}.RGB))
  TextureSampleParameter2D() {parameterName=Normal, samplerType=SAMPLERTYPE_Normal, texture=/Game/Environments/Galileo/Textures/T_Galileo_Stage_Cylinder_N}
  TextureSampleParameter2D() {parameterName=Spec/Metal/Rough, samplerType=SAMPLERTYPE_Masks, texture=/Game/Environments/Galileo/Textures/T_Galileo_Stage_Cylinder_S}
 ]
}
